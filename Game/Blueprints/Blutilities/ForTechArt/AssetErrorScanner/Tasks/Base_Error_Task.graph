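# Base editor utility task for the asset error scanner, with stub check functions that report no error by default.
# On execution it runs checkAssetForError on a valid asset, otherwise checkDataForError on valid asset data, storing the error flag and struct.
# It then stamps the error with its own class's package name, resolved through embedded Python, and finishes.
# A helper maps badness to a green-yellow-red color.
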EditorUtilityBlueprint /Game/Blueprints/Blutilities/ForTechArt/AssetErrorScanner/Tasks/Base_Error_Task
graphs [
 EventGraph {nodes [
   Event BeginExecution()
   > n1 = Sequence()
   n2 = Get asset() -> (asset:Object Reference)
   > n3 = Is Valid(InputObject=@n2)
   Branch(
    execute=@n3.Is Not Valid
    Condition=Is Valid Asset Data!(InAssetData:"Asset Data Structure (by ref)"=Get assetData() -> (assetData:Asset Data Structure)) -> (ReturnValue:bool)
   )
   > n4 = checkDataForError!(execute, self, assetData:Asset Data Structure=Get assetData()) -> (then, isError:bool, error:AVT Error Struct Structure)
   > Set Is Error(Is Error:bool=@n4) -> (Output_Get:bool)
   > Set finalErrorData(finalErrorData:AVT Error Struct Structure=@n4.error) -> (Output_Get:AVT Error Struct Structure)
   Set assetData(execute=@n3.Is Valid, assetData:Asset Data Structure=CreateAssetData(InAsset=@n2, bAllowBlueprintClass=false)) -> (Output_Get:Asset Data Structure)
   > n5 = checkAssetForError!(execute, self, asset:Object Reference=Get asset()) -> (then, isError:bool, error:AVT Error Struct Structure)
   > Set Is Error(Is Error=@n5)
   > Set finalErrorData(finalErrorData=@n5.error)
   n6 = getTaskPackageNameFromClass!(
    execute=@n1.then_1
    self
    class:Base Error Task Class Reference=Get Class(Object=Self-Reference()) -> (ReturnValue:Base Error Task Class Reference)
   ) -> (then, PackageName:Name=None)
   > n7 = Set members in AVT Error Struct!(
    execute:Exec
    StructRef:"AVT Error Struct Structure (by ref)"=Get finalErrorData() -> (finalErrorData:AVT Error Struct Structure)
    optionalOriginPackageName_62_#1:Name=@n6
   ) -> (then:Exec, StructOut:"AVT Error Struct Structure (by ref)") {is=title K2Node_SetFieldsInStruct}
   > Set finalErrorData(finalErrorData=@n7)
   > FinishExecutingTask()
  ]}
 checkAssetForError {nodes [
   checkAssetForError!() -> (then, asset:Object Reference) {is=title K2Node_FunctionEntry}
   > Return Node!(execute, isError:bool=false, error:AVT Error Struct Structure)
  ]}
 checkDataForError {nodes [
   checkDataForError!() -> (then, assetData:Asset Data Structure) {is=title K2Node_FunctionEntry}
   > Return Node(isError=false)
  ]}
 getErrorColor {nodes [
   n1 = getErrorColor!() -> (then, badness:double=~) {is=title K2Node_FunctionEntry}
   > Set l_badness(l_badness:double=@n1) -> (Output_Get:double)
   n2 = NormalizeToRange!(Value:double=Get l_badness() -> (l_badness:double), RangeMin:double=0.5, RangeMax:double=1) -> (ReturnValue:double)
   n3 = NormalizeToRange(Value=Get l_badness(), RangeMin=0, RangeMax=0.5)
   n4 = Get l_badness()
   > Return Node!(
    execute
    Color:Linear Color Structure=Select(
     Option 0:Linear Color Structure="Lerp (LinearColor)"(A=(R=0.0745,G=0.5312,B=0,A=1), B=(R=0.5255,G=0.4784,B=0,A=1), Alpha=@n3)
     Option 1:Linear Color Structure="Lerp (LinearColor)"(A=(R=0.5255,G=0.4784,B=0,A=1), B=(R=0.526,G=0,B=0,A=1), Alpha=@n2)
     Index="float > float"(A=@n4, B=0.5)
    ) -> (ReturnValue:Linear Color Structure)
   )
  ]}
 getFinalErrorData {nodes [
   getFinalErrorData() {is=title K2Node_FunctionEntry}
   > Return Node!(execute, isError:bool=Get Is Error() -> (Is Error:bool), errorData:AVT Error Struct Structure=Get finalErrorData())
  ]}
 getTaskPackageNameFromClass {nodes [
   n1 = getTaskPackageNameFromClass!() -> (then, class:Base Error Task Class Reference) {is=title K2Node_FunctionEntry}
   > n2 = Execute Python Script!(
    execute
    PythonScript="import unreal\r\nfinal_package_name = str(in_object).split(\"Object '\")[1]\r\nfinal_package_name = final_package_name.split(\".\")[0]\r\nfinal_asset = unreal.load_asset(final_package_name)\r\n##print(\"final_asset \" + str(final_asset))\r\nasset_data = unreal.AssetRegistryHelpers.create_asset_data(final_asset)\r\n##print(\"asset_data \" + str(asset_data))"
    in_object:Base Error Task Class Reference=@n1
   ) -> (then, ReturnValue, asset_data:"Asset Data Structure (by ref)") {inputs [in_object], outputs [asset_data]}
   > Return Node!(execute, PackageName:Name=Break Asset Data!(AssetData=@n2.asset_data) -> (PackageName, AssetClassPath))
  ]}
]
variables [
 asset {type=Object, default=null}
 assetData {type=AssetData, default {}}
 finalErrorData {
  type=AVT_Error_Struct
  default {
   csvHeader=""
   csvLine=""
   description=""
   helpCommand=""
   optionalOriginPackageName=None
   optionalTypeTag=""
   title=""
   tooltip=""
   color=<rgba 0 0 0 0>
   optionalAssetData {}
  }
 }
 friendlyName {type=string, default=Error Check}
 Is Error {type=boolean}
]
