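# Editor utility base task for the asset error scanner that decides whether a check runs on an asset, based on whether it is FX-only.
# validateFXAsset queries Asset Registry referencers filtered to World, Blueprint and NiagaraSystem, treating the asset as FX when all are Niagara systems.
# Otherwise it falls back to matching "FX" or "Effect" in the asset's name data.
# The cached shouldRunType result sets isFXAsset, while OnlyCheckFXAssets controls whether non-FX assets still run.
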EditorUtilityBlueprint /Game/Blueprints/Blutilities/ForTechArt/AssetErrorScanner/Tasks/Error_FXOnlyBase_Task
graphs [
 EventGraph {links [], nodes []}
 validateFXAsset {nodes [
   validateFXAsset() {is=title K2Node_FunctionEntry}
   > n1 = Sequence()
   > Branch(Condition="Equal (Enum)"(A:AES_FXMesh_Enum Enum=Get shouldRunType() -> (shouldRunType:AES_FXMesh_Enum Enum=NewEnumerator0), B:AES_FXMesh_Enum Enum=NewEnumerator0))
   > n2 = Sequence()
   > n3 = Sequence!(execute) -> (then_0, then_1, then_2:Exec)
   > n4 = Get Referencers!(
    execute
    self:Asset Registry Interface=GetAssetRegistry()
    PackageName:Name=Get!() -> (assetData_PackageName:Name, assetData_PackagePath:Name, assetData_AssetName:Name, assetData_AssetClassPath:Top Level Asset Path Structure)
    ReferenceOptions:"Asset Registry Dependency Options Structure (by ref)"=Make Asset Registry Dependency Options(bIncludeSoftPackageReferences=True)
   ) -> (then, OutReferencers:Array of Names, ReturnValue:bool) {is=title K2Node_CallFunction}
   > Branch(Condition=Is Valid Index!(TargetArray:Array of Names=@n4, IndexToTest:int) -> (ReturnValue:bool))
   > n5 = GetAssets!(
    execute
    self:Asset Registry Interface=GetAssetRegistry()
    Filter:"ARFilter Structure (by ref)"=MakeARFilter!(
     PackageNames:Array of Names=@n4
     PackagePaths:Array of Names
     SoftObjectPaths:Array of Soft Object Path Structures
     ClassPaths:Array of Top Level Asset Path Structures
     RecursiveClassPathsExclusionSet:Set of Top Level Asset Path Structures
     ClassNames:Array of Names=Make Array!("[0]":Name=World, "[1]":Name=Blueprint, "[2]":Name=NiagaraSystem) -> (Array:Array of Names)
     RecursiveClassesExclusionSet:Set of Names
     bRecursivePaths:bool=false
     bRecursiveClasses:bool=false
     bIncludeOnlyOnDiskAssets:bool=false
    ) -> (ReturnValue:ARFilter Structure)
    bSkipARFilteredAssets:bool=true
   ) -> (then, OutAssetData:Array of Asset Data Structures, ReturnValue:bool)
   > Set shouldRunType(shouldRunType:AES_FXMesh_Enum Enum=NewEnumerator1) -> (Output_Get:AES_FXMesh_Enum Enum=NewEnumerator0)
   > n6 = For Each Loop with Break!(Exec:Exec, Array:Array of Asset Data Structures=@n5, Break:Exec=@n7.then) -> (LoopBody:Exec, Array Element:Asset Data Structure, Array Index:int, Completed:Exec)
   > Branch(Condition="Not Equal (Class)"(
     A:Object Class Reference=GetClass!(InAssetData:"Asset Data Structure (by ref)"=@n6) -> (ReturnValue:Object Class Reference)
     B:Object Class Reference=/Script/Niagara.NiagaraSystem
    ) -> (ReturnValue:bool=~) {is=title K2Node_PromotableOperator})
   > n7 = Set shouldRunType(shouldRunType=NewEnumerator0) -> (Output_Get=NewEnumerator0)
   Branch(execute=@n3.then_1, Condition="Equal (Enum)"(A=Get shouldRunType() -> (shouldRunType=NewEnumerator0), B=NewEnumerator0))
   n8 = "To String (Name)"(InName=Break Asset Data(AssetData=Get assetData() -> (assetData:Asset Data Structure)))
   > Branch(Condition=OR Boolean(A=Contains(SearchIn=@n8, Substring=FX), B=Contains(SearchIn=@n8, Substring=Effect)))
   > Set shouldRunType(shouldRunType=NewEnumerator2) -> (Output_Get=NewEnumerator0)
   n9 = Get shouldRunType() -> (shouldRunType=NewEnumerator0)
   Set isFXAsset(execute=@n3.then_2, isFXAsset:bool=OR Boolean(A="Equal (Enum)"(A=@n9, B=NewEnumerator1), B="Equal (Enum)"(A=@n9, B=NewEnumerator2))) -> (Output_Get:bool)
   Branch(execute=@n2.then_1, Condition="Equal (Enum)"(A=Get shouldRunType() -> (shouldRunType=NewEnumerator0), B=NewEnumerator0))
   > n10 = Branch(Condition=Get OnlyCheckFXAssets() -> (OnlyCheckFXAssets:bool))
   > Set shouldRunType(shouldRunType=NewEnumerator4) -> (Output_Get=NewEnumerator0)
   Set shouldRunType(execute=@n10.else, shouldRunType=NewEnumerator3) -> (Output_Get=NewEnumerator0)
   n11 = Get shouldRunType() -> (shouldRunType=NewEnumerator0)
   Return Node!(
    execute=@n1.then_1
    shouldRun:bool=AND Boolean(
     A="Not Equal (Enum)"!(A:AES_FXMesh_Enum Enum=@n11, B:AES_FXMesh_Enum Enum=NewEnumerator0) -> (ReturnValue:bool)
     B="Not Equal (Enum)"(A=@n11, B=NewEnumerator4)
    )
   )
  ]}
]
variables [
 isFXAsset {type=boolean}
 OnlyCheckFXAssets {type=boolean, default=true}
 shouldRunType {type=AES_FXMesh_Enum, default=Unknown}
]
